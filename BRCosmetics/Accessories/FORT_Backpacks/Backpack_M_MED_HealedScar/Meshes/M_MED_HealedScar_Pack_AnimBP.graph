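# Anim Blueprint for the HealedScar backpack cosmetic, placing the pack on the wearer and animating its leaf, rope, bow and ribbons.
# A shared start pose sets spine bones and simulates the leaf with AnimDynamics.
# Body variants then re-offset the pack and chain AnimDynamics, using planar or spherical limits on spine and pelvis bones against clipping.
# Bool blends pick by gender and size, with extra male front-end offsets.
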AnimBlueprint /BRCosmetics/Accessories/FORT_Backpacks/Backpack_M_MED_HealedScar/Meshes/M_MED_HealedScar_Pack_AnimBP
target_skeleton=&/M_MED_HealedScar_Pack_Skeleton
graphs [
 AnimGraph {nodes [
   n1 = "Transform (Modify) Bone - Bone: pack"(ComponentPose=Local To Component(LocalPose=Use cached pose 'male'()), Translation="0,0,5", Rotation="0,0,0") {
    translationMode=BMM_Additive
    translationSpace=BCS_BoneSpace
    alphaBoolBlend {blendOption=Linear}
    boneToModify {boneName=pack}
    translation {z=5}
   }
   n2 = "Transform (Modify) Bone - Bone: dyn_rope_01"(ComponentPose=@n1, Alpha=1, Translation="-0.8,0,0", Rotation="0,0,0", Scale="1,1,1") {
    translationMode=BMM_Additive
    translationSpace=BCS_BoneSpace
    alphaBoolBlend {blendOption=Linear}
    boneToModify {boneName=dyn_rope_01}
    translation {x=-0.8}
   }
   n3 = Component To Local(ComponentPose=@n2)
   n4 = Use cached pose 'male'()
   n5 = Get bIsInFrontEnd()
   Output Pose(Result=Blend Poses by bool(
     BlendPose_0=Blend Poses by bool(BlendPose_0=@n3, BlendPose_1=@n4, bActiveValue=@n5) {blendType=Linear}
     BlendPose_1=Use cached pose 'female '()
     bActiveValue=Get isMale() -> (isMale:bool)
    ) {blendType=Linear}) {layerGroup=None}
   n6 = "Transform (Modify) Bone - Bone: pelvis_in"!(ComponentPose=Local To Component(LocalPose=Input Pose()), Alpha=1, Translation="0,0,0", Rotation="0,0,0") -> (Pose) {
    rotationMode=BMM_Replace
    rotationSpace=BCS_WorldSpace
    translationMode=BMM_Replace
    translationSpace=BCS_WorldSpace
    alphaBoolBlend {blendOption=Linear}
    boneToModify {boneName=pelvis_in}
   }
   n7 = "Transform (Modify) Bone - Bone: spine_01_in"!(ComponentPose=@n6, Alpha=1, Translation="0,0,0", Rotation="0,0,0") -> (Pose) {
    rotationMode=BMM_Replace
    rotationSpace=BCS_WorldSpace
    translationMode=BMM_Replace
    translationSpace=BCS_WorldSpace
    alphaBoolBlend {blendOption=Linear}
    boneToModify {boneName=spine_01_in}
   }
   n8 = "Transform (Modify) Bone - Bone: spine_02_in"!(ComponentPose=@n7, Alpha=1, Translation="0,0,0", Rotation="0,0,0") -> (Pose) {
    rotationMode=BMM_Replace
    rotationSpace=BCS_WorldSpace
    translationMode=BMM_Replace
    translationSpace=BCS_WorldSpace
    alphaBoolBlend {blendOption=Linear}
    boneToModify {boneName=spine_02_in}
   }
   n9 = "Transform (Modify) Bone - Bone: spine_03_in"!(ComponentPose=@n8, Alpha=1, Translation="0,0,0", Rotation="0,0,0") -> (Pose) {
    rotationMode=BMM_Replace
    rotationSpace=BCS_WorldSpace
    translationMode=BMM_Replace
    translationSpace=BCS_WorldSpace
    alphaBoolBlend {blendOption=Linear}
    boneToModify {boneName=spine_03_in}
   }
   n10 = "Transform (Modify) Bone - Bone: spine_04_in"!(ComponentPose=@n9, Alpha=1, Translation="0,0,0", Rotation="0,0,0") -> (Pose) {
    rotationMode=BMM_Replace
    rotationSpace=BCS_WorldSpace
    translationMode=BMM_Replace
    translationSpace=BCS_WorldSpace
    alphaBoolBlend {blendOption=Linear}
    boneToModify {boneName=spine_04_in}
   }
   n11 = "Transform (Modify) Bone - Bone: pack"!(ComponentPose=@n10, Alpha, Translation="0,0,0", Rotation="0,0,5") -> (Pose) {
    rotationMode=BMM_Additive
    rotationSpace=BCS_BoneSpace
    translationMode=BMM_Additive
    translationSpace=BCS_BoneSpace
    alphaBoolBlend {blendOption=Linear}
    boneToModify {boneName=pack}
    rotation {roll=5}
   }
   n12 = "Transform (Modify) Bone - Bone: pack"(ComponentPose=@n11, Translation="0,0,0", Rotation="0,0,5") {
    rotationMode=BMM_Additive
    rotationSpace=BCS_BoneSpace
    translationMode=BMM_Additive
    translationSpace=BCS_BoneSpace
    alphaBoolBlend {blendOption=Linear}
    boneToModify {boneName=pack}
    rotation {roll=5}
   }
   n13 = Anim Dynamics - Bone: dyn_leaf!(ComponentPose=@n12, Alpha=1, GravityOverride:Vector="0,0,-500") -> (Pose) {
    angularBiasOverride=0.999
    angularDampingOverride=0.999
    bOverrideAngularBias=true
    bOverrideAngularDamping=true
    bUseGravityOverride=true
    bUsePlanarLimit=false
    simulationSpace=RootRelative
    alphaBoolBlend {blendOption=Linear}
    boundBone {boneName=dyn_leaf}
    componentAppliedLinearAccClamp=<1000 1000 1000>
    componentLinearAccScale=<0.5 0.5 0.5>
    componentLinearVelScale=<0.5 0.5 0.5>
    physicsBodyDefinitions [{
      sphereCollisionRadius=2
      boundBone {boneName=dyn_leaf}
      boxExtents=<20 20 20>
      constraintSetup {angularLimitsMax=<0 0 0>, angularLimitsMin=<0 -85 0>, angularTarget=<1 0 0>}
      localJointOffset=<2 0 0>
     }]
   }
   Save cached pose 'start pose'(Pose=Component To Local(ComponentPose=@n13)) {cacheName=start pose}
   n14 = "Transform (Modify) Bone - Bone: pack"(ComponentPose=Local To Component(LocalPose=Use cached pose 'start pose'()), Translation="0,-1,0", Rotation="0,0,0") {
    rotationMode=BMM_Additive
    rotationSpace=BCS_BoneSpace
    translationMode=BMM_Additive
    translationSpace=BCS_BoneSpace
    alphaBoolBlend {blendOption=Linear}
    boneToModify {boneName=pack}
    translation {y=-1}
   }
   n15 = "Transform (Modify) Bone - Bone: pack"(ComponentPose=@n14, Translation="0,-4,0", Rotation="0,0,0") {
    rotationMode=BMM_Additive
    rotationSpace=BCS_BoneSpace
    translationMode=BMM_Replace
    translationSpace=BCS_BoneSpace
    alphaBoolBlend {blendOption=Linear}
    boneToModify {boneName=pack}
    translation {y=-4}
   }
   n16 = "Transform (Modify) Bone - Bone: pack"(ComponentPose=@n15, Translation="0,-7,5", Rotation="0,0,60") {
    rotationMode=BMM_Additive
    rotationSpace=BCS_BoneSpace
    translationMode=BMM_Replace
    translationSpace=BCS_BoneSpace
    alphaBoolBlend {blendOption=Linear}
    boneToModify {boneName=pack}
    rotation {roll=60}
    translation {y=-7, z=5}
   }
   n17 = Anim Dynamics - Bone: dyn_rope_01!(
    ComponentPose=@n16
    Alpha=1
    GravityOverride:Vector="0,0,0"
    ComponentLinearAccScale:Vector="0.5,0.5,0.5"
    ComponentLinearVelScale:Vector="0.5,0.5,0.5"
   ) -> (Pose) {
    angularBiasOverride=1
    angularDampingOverride=1
    angularSpringConstant=500
    bUseGravityOverride=true
    bUsePlanarLimit=false
    simulationSpace=RootRelative
    alphaBoolBlend {blendOption=Linear}
    boundBone {boneName=dyn_rope_01}
    componentAppliedLinearAccClamp=<1000 1000 1000>
    componentLinearAccScale=<0.5 0.5 0.5>
    componentLinearVelScale=<0.5 0.5 0.5>
    physicsBodyDefinitions [{
      collisionType=CustomSphere
      sphereCollisionRadius=2
      boundBone {boneName=dyn_rope_01}
      boxExtents=<25 25 5>
      constraintSetup {angularLimitsMax=<0 2.5 0>, angularLimitsMin=<0 0 0>, angularTarget=<1 0 0>}
      localJointOffset=<2 0 0>
     }]
   }
   n18 = Anim Dynamics - Bone: dyn_rope_02!(
    ComponentPose=@n17
    Alpha=1
    GravityOverride:Vector="0,0,0"
    ComponentLinearAccScale:Vector="0.5,0.5,0.5"
    ComponentLinearVelScale:Vector="0.5,0.5,0.5"
   ) -> (Pose) {
    angularBiasOverride=1
    angularDampingOverride=1
    angularSpringConstant=500
    bOverrideAngularBias=true
    bOverrideAngularDamping=true
    bUseGravityOverride=true
    simulationSpace=RootRelative
    alphaBoolBlend {blendOption=Linear}
    boundBone {boneName=dyn_rope_02}
    componentAppliedLinearAccClamp=<1000 1000 1000>
    componentLinearAccScale=<0.5 0.5 0.5>
    componentLinearVelScale=<0.5 0.5 0.5>
    physicsBodyDefinitions [{
      collisionType=CustomSphere
      sphereCollisionRadius=2
      boundBone {boneName=dyn_rope_02}
      boxExtents=<15 15 5>
      constraintSetup {angularLimitsMax=<0 15 0>, angularLimitsMin=<0 -45 0>, angularTarget=<1 0 0>}
      localJointOffset=<2 0 0>
     }]
    planarLimits [{drivingBone {boneName=spine_03_in}, planeTransform {location=<0 -8 0>, rotation=<rot 0 0 -90>}}]
   }
   n19 = Anim Dynamics - Bone: dyn_bow_01_L!(
    ComponentPose=@n18
    Alpha=1
    GravityOverride:Vector="0,0,0"
    ComponentLinearAccScale:Vector="0.5,0.5,0.5"
    ComponentLinearVelScale:Vector="0.5,0.5,0.5"
   ) -> (Pose) {
    angularBiasOverride=1
    angularDampingOverride=1
    bOverrideAngularBias=true
    bOverrideAngularDamping=true
    bUseGravityOverride=true
    bUsePlanarLimit=false
    simulationSpace=RootRelative
    alphaBoolBlend {blendOption=Linear}
    boundBone {boneName=dyn_bow_01_L}
    componentAppliedLinearAccClamp=<1000 1000 1000>
    componentLinearAccScale=<0.5 0.5 0.5>
    componentLinearVelScale=<0.5 0.5 0.5>
    physicsBodyDefinitions [{
      sphereCollisionRadius=1
      boundBone {boneName=dyn_bow_01_L}
      boxExtents=<15 15 15>
      constraintSetup {angularLimitsMax=<0 0 0>, angularLimitsMin=<0 -10 0>, angularTarget=<1 0 0>}
      localJointOffset=<2 0 0>
     }]
   }
   n20 = Anim Dynamics - Bone: dyn_bow!(
    ComponentPose=@n19
    Alpha=1
    GravityOverride:Vector="0,0,0"
    ComponentLinearAccScale:Vector="0.5,0.5,0.5"
    ComponentLinearVelScale:Vector="0.5,0.5,0.5"
   ) -> (Pose) {
    angularBiasOverride=1
    angularDampingOverride=1
    bOverrideAngularBias=true
    bOverrideAngularDamping=true
    bUseGravityOverride=true
    simulationSpace=RootRelative
    alphaBoolBlend {blendOption=Linear}
    boundBone {boneName=dyn_bow}
    componentAppliedLinearAccClamp=<1000 1000 1000>
    componentLinearAccScale=<0.5 0.5 0.5>
    componentLinearVelScale=<0.5 0.5 0.5>
    physicsBodyDefinitions [{
      collisionType=CustomSphere
      sphereCollisionRadius=2
      boundBone {boneName=dyn_bow}
      boxExtents=<10 10 5>
      constraintSetup {angularLimitsMax=<0 0 0>, angularLimitsMin=<0 -25 0>, angularTarget=<1 0 0>}
      localJointOffset=<1 0 0>
     }]
    planarLimits [{drivingBone {boneName=spine_03_in}, planeTransform {location=<0 -8 0>, rotation=<rot 0 0 -90>}}]
   }
   n21 = Anim Dynamics - Bone: dyn_bow_01_R!(
    ComponentPose=@n20
    Alpha=1
    GravityOverride:Vector="0,0,0"
    ComponentLinearAccScale:Vector="0.5,0.5,0.5"
    ComponentLinearVelScale:Vector="0.5,0.5,0.5"
   ) -> (Pose) {
    angularBiasOverride=1
    angularDampingOverride=1
    bOverrideAngularBias=true
    bOverrideAngularDamping=true
    bUseGravityOverride=true
    bUsePlanarLimit=false
    simulationSpace=RootRelative
    alphaBoolBlend {blendOption=Linear}
    boundBone {boneName=dyn_bow_01_R}
    componentAppliedLinearAccClamp=<1000 1000 1000>
    componentLinearAccScale=<0.5 0.5 0.5>
    componentLinearVelScale=<0.5 0.5 0.5>
    physicsBodyDefinitions [{
      sphereCollisionRadius=1
      boundBone {boneName=dyn_bow_01_R}
      boxExtents=<15 15 15>
      constraintSetup {angularLimitsMax=<0 0 0>, angularLimitsMin=<0 -10 0>, angularTarget=<1 0 0>}
      localJointOffset=<2 0 0>
     }]
   }
   n22 = Anim Dynamics - Bone: dyn_ribbon_01_L!(
    ComponentPose=@n21
    Alpha=1
    GravityOverride:Vector="0,0,0"
    ComponentLinearAccScale:Vector="0.5,0.5,0.5"
    ComponentLinearVelScale:Vector="0.5,0.5,0.5"
   ) -> (Pose) {
    angularBiasOverride=1
    angularDampingOverride=1
    bOverrideAngularBias=true
    bOverrideAngularDamping=true
    bUseGravityOverride=true
    bUsePlanarLimit=false
    bUseSphericalLimits=true
    simulationSpace=RootRelative
    alphaBoolBlend {blendOption=Linear}
    boundBone {boneName=dyn_ribbon_01_L}
    componentAppliedLinearAccClamp=<1000 1000 1000>
    componentLinearAccScale=<0.5 0.5 0.5>
    componentLinearVelScale=<0.5 0.5 0.5>
    physicsBodyDefinitions [{
      collisionType=CustomSphere
      sphereCollisionRadius=2
      boundBone {boneName=dyn_ribbon_01_L}
      boxExtents=<10 10 10>
      constraintSetup {angularLimitsMax=<0 0 10>, angularLimitsMin=<0 -25 0>, angularTarget=<1 0 0>}
      localJointOffset=<3 0 0>
     }]
    sphericalLimits [{limitRadius=12, drivingBone {boneName=spine_01_in}, sphereLocalOffset=<0 0 0>}]
   }
   n23 = Anim Dynamics - Bone: dyn_ribbon_01_R!(
    ComponentPose=@n22
    Alpha=1
    GravityOverride:Vector="0,0,0"
    ComponentLinearAccScale:Vector="0.5,0.5,0.5"
    ComponentLinearVelScale:Vector="0.5,0.5,0.5"
   ) -> (Pose) {
    angularBiasOverride=1
    angularDampingOverride=1
    bOverrideAngularBias=true
    bOverrideAngularDamping=true
    bUseGravityOverride=true
    bUsePlanarLimit=false
    bUseSphericalLimits=true
    simulationSpace=RootRelative
    alphaBoolBlend {blendOption=Linear}
    boundBone {boneName=dyn_ribbon_01_R}
    componentAppliedLinearAccClamp=<1000 1000 1000>
    componentLinearAccScale=<0.5 0.5 0.5>
    componentLinearVelScale=<0.5 0.5 0.5>
    physicsBodyDefinitions [{
      collisionType=CustomSphere
      sphereCollisionRadius=2
      boundBone {boneName=dyn_ribbon_01_R}
      boxExtents=<10 10 10>
      constraintSetup {angularLimitsMax=<0 0 0>, angularLimitsMin=<0 -25 -10>, angularTarget=<0 0 0>}
      localJointOffset=<3 0 0>
     }]
    sphericalLimits [{limitRadius=12, drivingBone {boneName=spine_01_in}, sphereLocalOffset=<0 0 0>}]
   }
   n24 = Anim Dynamics - Bone: dyn_ribbon_02_L!(
    ComponentPose=@n23
    Alpha=1
    GravityOverride:Vector="0,0,0"
    ComponentLinearAccScale:Vector="0.5,0.5,0.5"
    ComponentLinearVelScale:Vector="0.5,0.5,0.5"
   ) -> (Pose) {
    angularBiasOverride=1
    angularDampingOverride=1
    bOverrideAngularBias=true
    bOverrideAngularDamping=true
    bUseGravityOverride=true
    bUsePlanarLimit=false
    bUseSphericalLimits=true
    simulationSpace=RootRelative
    alphaBoolBlend {blendOption=Linear}
    boundBone {boneName=dyn_ribbon_02_L}
    componentAppliedLinearAccClamp=<1000 1000 1000>
    componentLinearAccScale=<0.5 0.5 0.5>
    componentLinearVelScale=<0.5 0.5 0.5>
    physicsBodyDefinitions [{
      collisionType=CustomSphere
      sphereCollisionRadius=2
      boundBone {boneName=dyn_ribbon_02_L}
      boxExtents=<10 10 10>
      constraintSetup {angularLimitsMax=<0 0 10>, angularLimitsMin=<0 -25 0>, angularTarget=<1 0 0>}
      localJointOffset=<4 0 0>
     }]
    sphericalLimits [{limitRadius=17, drivingBone {boneName=pelvis_in}, sphereLocalOffset=<0 1 0>}]
   }
   n25 = Anim Dynamics - Bone: dyn_ribbon_02_R!(
    ComponentPose=@n24
    Alpha=1
    GravityOverride:Vector="0,0,0"
    ComponentLinearAccScale:Vector="0.5,0.5,0.5"
    ComponentLinearVelScale:Vector="0.5,0.5,0.5"
   ) -> (Pose) {
    angularBiasOverride=1
    angularDampingOverride=1
    bOverrideAngularBias=true
    bOverrideAngularDamping=true
    bUseGravityOverride=true
    bUsePlanarLimit=false
    bUseSphericalLimits=true
    simulationSpace=RootRelative
    alphaBoolBlend {blendOption=Linear}
    boundBone {boneName=dyn_ribbon_02_R}
    componentAppliedLinearAccClamp=<1000 1000 1000>
    componentLinearAccScale=<0.5 0.5 0.5>
    componentLinearVelScale=<0.5 0.5 0.5>
    physicsBodyDefinitions [{
      collisionType=CustomSphere
      sphereCollisionRadius=2
      boundBone {boneName=dyn_ribbon_02_R}
      boxExtents=<10 10 10>
      constraintSetup {angularLimitsMax=<0 0 0>, angularLimitsMin=<0 -25 -10>, angularTarget=<0 0 0>}
      localJointOffset=<4 0 0>
     }]
    sphericalLimits [{limitRadius=17, drivingBone {boneName=pelvis_in}, sphereLocalOffset=<0 1 0>}]
   }
   n26 = "Transform (Modify) Bone - Bone: pack"(ComponentPose=Local To Component(LocalPose=Use cached pose 'start pose'()), Translation="0,-1,0", Rotation="0,0,0") {
    rotationMode=BMM_Additive
    rotationSpace=BCS_BoneSpace
    translationMode=BMM_Additive
    translationSpace=BCS_BoneSpace
    alphaBoolBlend {blendOption=Linear}
    boneToModify {boneName=pack}
    translation {y=-1}
   }
   n27 = "Transform (Modify) Bone - Bone: pack"(ComponentPose=@n26, Translation="0,-4,0", Rotation="0,0,0") {
    rotationMode=BMM_Additive
    rotationSpace=BCS_BoneSpace
    translationMode=BMM_Replace
    translationSpace=BCS_BoneSpace
    alphaBoolBlend {blendOption=Linear}
    boneToModify {boneName=pack}
    translation {y=-4}
   }
   n28 = "Transform (Modify) Bone - Bone: pack"(ComponentPose=@n27, Translation="0,-7,5", Rotation="0,0,60") {
    rotationMode=BMM_Additive
    rotationSpace=BCS_BoneSpace
    translationMode=BMM_Replace
    translationSpace=BCS_BoneSpace
    alphaBoolBlend {blendOption=Linear}
    boneToModify {boneName=pack}
    rotation {roll=60}
    translation {y=-7, z=5}
   }
   n29 = Anim Dynamics - Bone: dyn_rope_01(ComponentPose=@n28, Alpha=1, GravityOverride="0,0,0", ComponentLinearAccScale="0.5,0.5,0.5", ComponentLinearVelScale="0.5,0.5,0.5") {
    angularBiasOverride=1
    angularDampingOverride=1
    angularSpringConstant=500
    bUseGravityOverride=true
    bUsePlanarLimit=false
    simulationSpace=RootRelative
    alphaBoolBlend {blendOption=Linear}
    boundBone {boneName=dyn_rope_01}
    componentAppliedLinearAccClamp=<1000 1000 1000>
    componentLinearAccScale=<0.5 0.5 0.5>
    componentLinearVelScale=<0.5 0.5 0.5>
    physicsBodyDefinitions [{
      collisionType=CustomSphere
      sphereCollisionRadius=2
      boundBone {boneName=dyn_rope_01}
      boxExtents=<25 25 5>
      constraintSetup {angularLimitsMax=<0 2.5 0>, angularLimitsMin=<0 0 0>, angularTarget=<1 0 0>}
      localJointOffset=<2 0 0>
     }]
   }
   n30 = Anim Dynamics - Bone: dyn_rope_02(ComponentPose=@n29, Alpha=1, GravityOverride="0,0,0", ComponentLinearAccScale="0.5,0.5,0.5", ComponentLinearVelScale="0.5,0.5,0.5") {
    angularBiasOverride=1
    angularDampingOverride=1
    angularSpringConstant=500
    bOverrideAngularBias=true
    bOverrideAngularDamping=true
    bUseGravityOverride=true
    simulationSpace=RootRelative
    alphaBoolBlend {blendOption=Linear}
    boundBone {boneName=dyn_rope_02}
    componentAppliedLinearAccClamp=<1000 1000 1000>
    componentLinearAccScale=<0.5 0.5 0.5>
    componentLinearVelScale=<0.5 0.5 0.5>
    physicsBodyDefinitions [{
      collisionType=CustomSphere
      sphereCollisionRadius=2
      boundBone {boneName=dyn_rope_02}
      boxExtents=<15 15 5>
      constraintSetup {angularLimitsMax=<0 15 0>, angularLimitsMin=<0 -45 0>, angularTarget=<1 0 0>}
      localJointOffset=<2 0 0>
     }]
    planarLimits [{drivingBone {boneName=spine_03_in}, planeTransform {location=<0 -15 0>, rotation=<rot 0 0 -90>}}]
   }
   n31 = Anim Dynamics - Bone: dyn_bow_01_L(ComponentPose=@n30, Alpha=1, GravityOverride="0,0,0", ComponentLinearAccScale="0.5,0.5,0.5", ComponentLinearVelScale="0.5,0.5,0.5") {
    angularBiasOverride=1
    angularDampingOverride=1
    bOverrideAngularBias=true
    bOverrideAngularDamping=true
    bUseGravityOverride=true
    bUsePlanarLimit=false
    simulationSpace=RootRelative
    alphaBoolBlend {blendOption=Linear}
    boundBone {boneName=dyn_bow_01_L}
    componentAppliedLinearAccClamp=<1000 1000 1000>
    componentLinearAccScale=<0.5 0.5 0.5>
    componentLinearVelScale=<0.5 0.5 0.5>
    physicsBodyDefinitions [{
      sphereCollisionRadius=1
      boundBone {boneName=dyn_bow_01_L}
      boxExtents=<15 15 15>
      constraintSetup {angularLimitsMax=<0 0 0>, angularLimitsMin=<0 -10 0>, angularTarget=<1 0 0>}
      localJointOffset=<2 0 0>
     }]
   }
   n32 = Anim Dynamics - Bone: dyn_bow(ComponentPose=@n31, Alpha=1, GravityOverride="0,0,0", ComponentLinearAccScale="0.5,0.5,0.5", ComponentLinearVelScale="0.5,0.5,0.5") {
    angularBiasOverride=1
    angularDampingOverride=1
    bOverrideAngularBias=true
    bOverrideAngularDamping=true
    bUseGravityOverride=true
    simulationSpace=RootRelative
    alphaBoolBlend {blendOption=Linear}
    boundBone {boneName=dyn_bow}
    componentAppliedLinearAccClamp=<1000 1000 1000>
    componentLinearAccScale=<0.5 0.5 0.5>
    componentLinearVelScale=<0.5 0.5 0.5>
    physicsBodyDefinitions [{
      collisionType=CustomSphere
      sphereCollisionRadius=2
      boundBone {boneName=dyn_bow}
      boxExtents=<10 10 5>
      constraintSetup {angularLimitsMax=<0 0 0>, angularLimitsMin=<0 -25 0>, angularTarget=<1 0 0>}
      localJointOffset=<1 0 0>
     }]
    planarLimits [{drivingBone {boneName=spine_03_in}, planeTransform {location=<0 -16 0>, rotation=<rot 0 0 -90>}}]
   }
   n33 = Anim Dynamics - Bone: dyn_bow_01_R(ComponentPose=@n32, Alpha=1, GravityOverride="0,0,0", ComponentLinearAccScale="0.5,0.5,0.5", ComponentLinearVelScale="0.5,0.5,0.5") {
    angularBiasOverride=1
    angularDampingOverride=1
    bOverrideAngularBias=true
    bOverrideAngularDamping=true
    bUseGravityOverride=true
    bUsePlanarLimit=false
    simulationSpace=RootRelative
    alphaBoolBlend {blendOption=Linear}
    boundBone {boneName=dyn_bow_01_R}
    componentAppliedLinearAccClamp=<1000 1000 1000>
    componentLinearAccScale=<0.5 0.5 0.5>
    componentLinearVelScale=<0.5 0.5 0.5>
    physicsBodyDefinitions [{
      sphereCollisionRadius=1
      boundBone {boneName=dyn_bow_01_R}
      boxExtents=<15 15 15>
      constraintSetup {angularLimitsMax=<0 0 0>, angularLimitsMin=<0 -10 0>, angularTarget=<1 0 0>}
      localJointOffset=<2 0 0>
     }]
   }
   n34 = Anim Dynamics - Bone: dyn_ribbon_01_L(ComponentPose=@n33, Alpha=1, GravityOverride="0,0,0", ComponentLinearAccScale="0.5,0.5,0.5", ComponentLinearVelScale="0.5,0.5,0.5") {
    angularBiasOverride=1
    angularDampingOverride=1
    bOverrideAngularBias=true
    bOverrideAngularDamping=true
    bUseGravityOverride=true
    bUsePlanarLimit=false
    bUseSphericalLimits=true
    simulationSpace=RootRelative
    alphaBoolBlend {blendOption=Linear}
    boundBone {boneName=dyn_ribbon_01_L}
    componentAppliedLinearAccClamp=<1000 1000 1000>
    componentLinearAccScale=<0.5 0.5 0.5>
    componentLinearVelScale=<0.5 0.5 0.5>
    physicsBodyDefinitions [{
      collisionType=CustomSphere
      sphereCollisionRadius=2
      boundBone {boneName=dyn_ribbon_01_L}
      boxExtents=<10 10 10>
      constraintSetup {angularLimitsMax=<0 0 10>, angularLimitsMin=<0 -25 0>, angularTarget=<1 0 0>}
      localJointOffset=<3 0 0>
     }]
    sphericalLimits [{limitRadius=17, drivingBone {boneName=spine_02_in}, sphereLocalOffset=<0 0 0>}]
   }
   n35 = Anim Dynamics - Bone: dyn_ribbon_01_R(ComponentPose=@n34, Alpha=1, GravityOverride="0,0,0", ComponentLinearAccScale="0.5,0.5,0.5", ComponentLinearVelScale="0.5,0.5,0.5") {
    angularBiasOverride=1
    angularDampingOverride=1
    bOverrideAngularBias=true
    bOverrideAngularDamping=true
    bUseGravityOverride=true
    bUsePlanarLimit=false
    bUseSphericalLimits=true
    simulationSpace=RootRelative
    alphaBoolBlend {blendOption=Linear}
    boundBone {boneName=dyn_ribbon_01_R}
    componentAppliedLinearAccClamp=<1000 1000 1000>
    componentLinearAccScale=<0.5 0.5 0.5>
    componentLinearVelScale=<0.5 0.5 0.5>
    physicsBodyDefinitions [{
      collisionType=CustomSphere
      sphereCollisionRadius=2
      boundBone {boneName=dyn_ribbon_01_R}
      boxExtents=<10 10 10>
      constraintSetup {angularLimitsMax=<0 0 0>, angularLimitsMin=<0 -25 -10>, angularTarget=<0 0 0>}
      localJointOffset=<3 0 0>
     }]
    sphericalLimits [{limitRadius=17, drivingBone {boneName=spine_02_in}, sphereLocalOffset=<0 0 0>}]
   }
   n36 = Anim Dynamics - Bone: dyn_ribbon_02_L(ComponentPose=@n35, Alpha=1, GravityOverride="0,0,0", ComponentLinearAccScale="0.5,0.5,0.5", ComponentLinearVelScale="0.5,0.5,0.5") {
    angularBiasOverride=1
    angularDampingOverride=1
    bOverrideAngularBias=true
    bOverrideAngularDamping=true
    bUseGravityOverride=true
    bUsePlanarLimit=false
    bUseSphericalLimits=true
    simulationSpace=RootRelative
    alphaBoolBlend {blendOption=Linear}
    boundBone {boneName=dyn_ribbon_02_L}
    componentAppliedLinearAccClamp=<1000 1000 1000>
    componentLinearAccScale=<0.5 0.5 0.5>
    componentLinearVelScale=<0.5 0.5 0.5>
    physicsBodyDefinitions [{
      collisionType=CustomSphere
      sphereCollisionRadius=2
      boundBone {boneName=dyn_ribbon_02_L}
      boxExtents=<10 10 10>
      constraintSetup {angularLimitsMax=<0 0 10>, angularLimitsMin=<0 -25 0>, angularTarget=<1 0 0>}
      localJointOffset=<4 0 0>
     }]
    sphericalLimits [{limitRadius=18, drivingBone {boneName=spine_01_in}, sphereLocalOffset=<0 0 0>}]
   }
   n37 = Anim Dynamics - Bone: dyn_ribbon_02_R(ComponentPose=@n36, Alpha=1, GravityOverride="0,0,0", ComponentLinearAccScale="0.5,0.5,0.5", ComponentLinearVelScale="0.5,0.5,0.5") {
    angularBiasOverride=1
    angularDampingOverride=1
    bOverrideAngularBias=true
    bOverrideAngularDamping=true
    bUseGravityOverride=true
    bUsePlanarLimit=false
    bUseSphericalLimits=true
    simulationSpace=RootRelative
    alphaBoolBlend {blendOption=Linear}
    boundBone {boneName=dyn_ribbon_02_R}
    componentAppliedLinearAccClamp=<1000 1000 1000>
    componentLinearAccScale=<0.5 0.5 0.5>
    componentLinearVelScale=<0.5 0.5 0.5>
    physicsBodyDefinitions [{
      collisionType=CustomSphere
      sphereCollisionRadius=2
      boundBone {boneName=dyn_ribbon_02_R}
      boxExtents=<10 10 10>
      constraintSetup {angularLimitsMax=<0 0 0>, angularLimitsMin=<0 -25 -10>, angularTarget=<0 0 0>}
      localJointOffset=<4 0 0>
     }]
    sphericalLimits [{limitRadius=18, drivingBone {boneName=spine_01_in}, sphereLocalOffset=<0 0 0>}]
   }
   Save cached pose 'female '(Pose=Blend Poses by bool(
     BlendPose_0=Component To Local(ComponentPose=@n37)
     BlendPose_1=Component To Local(ComponentPose=@n25)
     bActiveValue=Get Is Large() -> (Is Large:bool)
    ) {blendType=Linear}) {cacheName="female "}
   n38 = "Transform (Modify) Bone - Bone: pack"(ComponentPose=Local To Component(LocalPose=Use cached pose 'start pose'()), Translation="0,-1.5,0", Rotation="0,0,0") {
    rotationMode=BMM_Additive
    rotationSpace=BCS_BoneSpace
    translationMode=BMM_Additive
    translationSpace=BCS_BoneSpace
    alphaBoolBlend {blendOption=Linear}
    boneToModify {boneName=pack}
    translation {y=-1.5}
   }
   n39 = "Transform (Modify) Bone - Bone: pack"(ComponentPose=@n38, Translation="0,-1.5,0", Rotation="0,0,0") {
    rotationMode=BMM_Additive
    rotationSpace=BCS_BoneSpace
    translationMode=BMM_Replace
    translationSpace=BCS_BoneSpace
    alphaBoolBlend {blendOption=Linear}
    boneToModify {boneName=pack}
    translation {y=-1.5}
   }
   n40 = "Transform (Modify) Bone - Bone: pack"(ComponentPose=@n39, Translation="0,-7,7", Rotation="0,0,50") {
    rotationMode=BMM_Additive
    rotationSpace=BCS_BoneSpace
    translationMode=BMM_Replace
    translationSpace=BCS_BoneSpace
    alphaBoolBlend {blendOption=Linear}
    boneToModify {boneName=pack}
    rotation {roll=50}
    translation {y=-7, z=7}
   }
   n41 = Anim Dynamics - Bone: dyn_rope_01(ComponentPose=@n40, Alpha=1, GravityOverride="0,0,0", ComponentLinearAccScale="0.5,0.5,0.5", ComponentLinearVelScale="0.5,0.5,0.5") {
    angularBiasOverride=1
    angularDampingOverride=1
    angularSpringConstant=500
    bUseGravityOverride=true
    bUsePlanarLimit=false
    simulationSpace=RootRelative
    alphaBoolBlend {blendOption=Linear}
    boundBone {boneName=dyn_rope_01}
    componentAppliedLinearAccClamp=<1000 1000 1000>
    componentLinearAccScale=<0.5 0.5 0.5>
    componentLinearVelScale=<0.5 0.5 0.5>
    physicsBodyDefinitions [{
      collisionType=CustomSphere
      sphereCollisionRadius=2
      boundBone {boneName=dyn_rope_01}
      boxExtents=<25 25 5>
      constraintSetup {angularLimitsMax=<0 2.5 0>, angularLimitsMin=<0 0 0>, angularTarget=<1 0 0>}
      localJointOffset=<2 0 0>
     }]
   }
   n42 = Anim Dynamics - Bone: dyn_rope_02(ComponentPose=@n41, Alpha=1, GravityOverride="0,0,0", ComponentLinearAccScale="0.5,0.5,0.5", ComponentLinearVelScale="0.5,0.5,0.5") {
    angularBiasOverride=1
    angularDampingOverride=1
    angularSpringConstant=500
    bOverrideAngularBias=true
    bOverrideAngularDamping=true
    bUseGravityOverride=true
    simulationSpace=RootRelative
    alphaBoolBlend {blendOption=Linear}
    boundBone {boneName=dyn_rope_02}
    componentAppliedLinearAccClamp=<1000 1000 1000>
    componentLinearAccScale=<0.5 0.5 0.5>
    componentLinearVelScale=<0.5 0.5 0.5>
    physicsBodyDefinitions [{
      collisionType=CustomSphere
      sphereCollisionRadius=2
      boundBone {boneName=dyn_rope_02}
      boxExtents=<15 15 5>
      constraintSetup {angularLimitsMax=<0 20 0>, angularLimitsMin=<0 -45 0>, angularTarget=<1 0 0>}
      localJointOffset=<2 0 0>
     }]
    planarLimits [{drivingBone {boneName=spine_03_in}, planeTransform {location=<0 -11 0>, rotation=<rot 0 0 -90>}}]
   }
   n43 = Anim Dynamics - Bone: dyn_bow_01_L(ComponentPose=@n42, Alpha=1, GravityOverride="0,0,0", ComponentLinearAccScale="0.5,0.5,0.5", ComponentLinearVelScale="0.5,0.5,0.5") {
    angularBiasOverride=1
    angularDampingOverride=1
    bOverrideAngularBias=true
    bOverrideAngularDamping=true
    bUseGravityOverride=true
    bUsePlanarLimit=false
    simulationSpace=RootRelative
    alphaBoolBlend {blendOption=Linear}
    boundBone {boneName=dyn_bow_01_L}
    componentAppliedLinearAccClamp=<1000 1000 1000>
    componentLinearAccScale=<0.5 0.5 0.5>
    componentLinearVelScale=<0.5 0.5 0.5>
    physicsBodyDefinitions [{
      sphereCollisionRadius=1
      boundBone {boneName=dyn_bow_01_L}
      boxExtents=<15 15 15>
      constraintSetup {angularLimitsMax=<0 0 0>, angularLimitsMin=<0 -10 0>, angularTarget=<1 0 0>}
      localJointOffset=<2 0 0>
     }]
   }
   n44 = Anim Dynamics - Bone: dyn_bow(ComponentPose=@n43, Alpha=1, GravityOverride="0,0,0", ComponentLinearAccScale="0.5,0.5,0.5", ComponentLinearVelScale="0.5,0.5,0.5") {
    angularBiasOverride=1
    angularDampingOverride=1
    bOverrideAngularBias=true
    bOverrideAngularDamping=true
    bUseGravityOverride=true
    simulationSpace=RootRelative
    alphaBoolBlend {blendOption=Linear}
    boundBone {boneName=dyn_bow}
    componentAppliedLinearAccClamp=<1000 1000 1000>
    componentLinearAccScale=<0.5 0.5 0.5>
    componentLinearVelScale=<0.5 0.5 0.5>
    physicsBodyDefinitions [{
      collisionType=CustomSphere
      sphereCollisionRadius=2
      boundBone {boneName=dyn_bow}
      boxExtents=<10 10 5>
      constraintSetup {angularLimitsMax=<0 0 0>, angularLimitsMin=<0 -25 0>, angularTarget=<1 0 0>}
      localJointOffset=<1 0 0>
     }]
    planarLimits [{drivingBone {boneName=spine_03_in}, planeTransform {location=<0 -11 0>, rotation=<rot 0 0 -90>}}]
   }
   n45 = Anim Dynamics - Bone: dyn_bow_01_R(ComponentPose=@n44, Alpha=1, GravityOverride="0,0,0", ComponentLinearAccScale="0.5,0.5,0.5", ComponentLinearVelScale="0.5,0.5,0.5") {
    angularBiasOverride=1
    angularDampingOverride=1
    bOverrideAngularBias=true
    bOverrideAngularDamping=true
    bUseGravityOverride=true
    bUsePlanarLimit=false
    simulationSpace=RootRelative
    alphaBoolBlend {blendOption=Linear}
    boundBone {boneName=dyn_bow_01_R}
    componentAppliedLinearAccClamp=<1000 1000 1000>
    componentLinearAccScale=<0.5 0.5 0.5>
    componentLinearVelScale=<0.5 0.5 0.5>
    physicsBodyDefinitions [{
      sphereCollisionRadius=1
      boundBone {boneName=dyn_bow_01_R}
      boxExtents=<15 15 15>
      constraintSetup {angularLimitsMax=<0 0 0>, angularLimitsMin=<0 -10 0>, angularTarget=<1 0 0>}
      localJointOffset=<2 0 0>
     }]
   }
   n46 = Anim Dynamics - Bone: dyn_ribbon_01_L(ComponentPose=@n45, Alpha=1, GravityOverride="0,0,0", ComponentLinearAccScale="0.5,0.5,0.5", ComponentLinearVelScale="0.5,0.5,0.5") {
    angularBiasOverride=1
    angularDampingOverride=1
    bOverrideAngularBias=true
    bOverrideAngularDamping=true
    bUseGravityOverride=true
    simulationSpace=RootRelative
    alphaBoolBlend {blendOption=Linear}
    boundBone {boneName=dyn_ribbon_01_L}
    componentAppliedLinearAccClamp=<1000 1000 1000>
    componentLinearAccScale=<0.5 0.5 0.5>
    componentLinearVelScale=<0.5 0.5 0.5>
    physicsBodyDefinitions [{
      collisionType=CustomSphere
      sphereCollisionRadius=2
      boundBone {boneName=dyn_ribbon_01_L}
      boxExtents=<10 10 10>
      constraintSetup {angularLimitsMax=<0 0 10>, angularLimitsMin=<0 -25 0>, angularTarget=<1 0 0>}
      localJointOffset=<3 0 0>
     }]
    planarLimits [{drivingBone {boneName=spine_02_in}, planeTransform {location=<0 -10 0>, rotation=<rot 0 0 -90>}}]
   }
   n47 = Anim Dynamics - Bone: dyn_ribbon_01_R(ComponentPose=@n46, Alpha=1, GravityOverride="0,0,0", ComponentLinearAccScale="0.5,0.5,0.5", ComponentLinearVelScale="0.5,0.5,0.5") {
    angularBiasOverride=1
    angularDampingOverride=1
    bOverrideAngularBias=true
    bOverrideAngularDamping=true
    bUseGravityOverride=true
    simulationSpace=RootRelative
    alphaBoolBlend {blendOption=Linear}
    boundBone {boneName=dyn_ribbon_01_R}
    componentAppliedLinearAccClamp=<1000 1000 1000>
    componentLinearAccScale=<0.5 0.5 0.5>
    componentLinearVelScale=<0.5 0.5 0.5>
    physicsBodyDefinitions [{
      collisionType=CustomSphere
      sphereCollisionRadius=2
      boundBone {boneName=dyn_ribbon_01_R}
      boxExtents=<10 10 10>
      constraintSetup {angularLimitsMax=<0 0 0>, angularLimitsMin=<0 -25 -10>, angularTarget=<0 0 0>}
      localJointOffset=<3 0 0>
     }]
    planarLimits [{drivingBone {boneName=spine_02_in}, planeTransform {location=<0 -10 0>, rotation=<rot 0 0 -90>}}]
   }
   n48 = Anim Dynamics - Bone: dyn_ribbon_02_L(ComponentPose=@n47, Alpha=1, GravityOverride="0,0,0", ComponentLinearAccScale="0.5,0.5,0.5", ComponentLinearVelScale="0.5,0.5,0.5") {
    angularBiasOverride=1
    angularDampingOverride=1
    bOverrideAngularBias=true
    bOverrideAngularDamping=true
    bUseGravityOverride=true
    simulationSpace=RootRelative
    alphaBoolBlend {blendOption=Linear}
    boundBone {boneName=dyn_ribbon_02_L}
    componentAppliedLinearAccClamp=<1000 1000 1000>
    componentLinearAccScale=<0.5 0.5 0.5>
    componentLinearVelScale=<0.5 0.5 0.5>
    physicsBodyDefinitions [{
      collisionType=CustomSphere
      sphereCollisionRadius=2
      boundBone {boneName=dyn_ribbon_02_L}
      boxExtents=<10 10 10>
      constraintSetup {angularLimitsMax=<0 0 10>, angularLimitsMin=<0 -25 0>, angularTarget=<1 0 0>}
      localJointOffset=<4 0 0>
     }]
    planarLimits [{drivingBone {boneName=spine_02_in}, planeTransform {location=<0 -10 0>, rotation=<rot 0 0 -90>}}]
   }
   n49 = Anim Dynamics - Bone: dyn_ribbon_02_R(ComponentPose=@n48, Alpha=1, GravityOverride="0,0,0", ComponentLinearAccScale="0.5,0.5,0.5", ComponentLinearVelScale="0.5,0.5,0.5") {
    angularBiasOverride=1
    angularDampingOverride=1
    bOverrideAngularBias=true
    bOverrideAngularDamping=true
    bUseGravityOverride=true
    simulationSpace=RootRelative
    alphaBoolBlend {blendOption=Linear}
    boundBone {boneName=dyn_ribbon_02_R}
    componentAppliedLinearAccClamp=<1000 1000 1000>
    componentLinearAccScale=<0.5 0.5 0.5>
    componentLinearVelScale=<0.5 0.5 0.5>
    physicsBodyDefinitions [{
      collisionType=CustomSphere
      sphereCollisionRadius=2
      boundBone {boneName=dyn_ribbon_02_R}
      boxExtents=<10 10 10>
      constraintSetup {angularLimitsMax=<0 0 0>, angularLimitsMin=<0 -25 -10>, angularTarget=<0 0 0>}
      localJointOffset=<4 0 0>
     }]
    planarLimits [{drivingBone {boneName=spine_02_in}, planeTransform {location=<0 -10 0>, rotation=<rot 0 0 -90>}}]
   }
   n50 = "Transform (Modify) Bone - Bone: pack"(ComponentPose=Local To Component(LocalPose=Use cached pose 'start pose'()), Translation="0,-1.5,0", Rotation="0,0,0") {
    rotationMode=BMM_Additive
    rotationSpace=BCS_BoneSpace
    translationMode=BMM_Additive
    translationSpace=BCS_BoneSpace
    alphaBoolBlend {blendOption=Linear}
    boneToModify {boneName=pack}
    translation {y=-1.5}
   }
   n51 = "Transform (Modify) Bone - Bone: pack"(ComponentPose=@n50, Translation="0,-1.5,0", Rotation="0,0,0") {
    rotationMode=BMM_Additive
    rotationSpace=BCS_BoneSpace
    translationMode=BMM_Replace
    translationSpace=BCS_BoneSpace
    alphaBoolBlend {blendOption=Linear}
    boneToModify {boneName=pack}
    translation {y=-1.5}
   }
   n52 = "Transform (Modify) Bone - Bone: pack"(ComponentPose=@n51, Translation="0,-7,7", Rotation="0,0,50") {
    rotationMode=BMM_Additive
    rotationSpace=BCS_BoneSpace
    translationMode=BMM_Replace
    translationSpace=BCS_BoneSpace
    alphaBoolBlend {blendOption=Linear}
    boneToModify {boneName=pack}
    rotation {roll=50}
    translation {y=-7, z=7}
   }
   n53 = Anim Dynamics - Bone: dyn_rope_01(ComponentPose=@n52, Alpha=1, GravityOverride="0,0,0", ComponentLinearAccScale="0.5,0.5,0.5", ComponentLinearVelScale="0.5,0.5,0.5") {
    angularBiasOverride=1
    angularDampingOverride=1
    angularSpringConstant=500
    bUseGravityOverride=true
    bUsePlanarLimit=false
    simulationSpace=RootRelative
    alphaBoolBlend {blendOption=Linear}
    boundBone {boneName=dyn_rope_01}
    componentAppliedLinearAccClamp=<1000 1000 1000>
    componentLinearAccScale=<0.5 0.5 0.5>
    componentLinearVelScale=<0.5 0.5 0.5>
    physicsBodyDefinitions [{
      collisionType=CustomSphere
      sphereCollisionRadius=2
      boundBone {boneName=dyn_rope_01}
      boxExtents=<25 25 5>
      constraintSetup {angularLimitsMax=<0 2.5 0>, angularLimitsMin=<0 0 0>, angularTarget=<1 0 0>}
      localJointOffset=<2 0 0>
     }]
   }
   n54 = Anim Dynamics - Bone: dyn_rope_02(ComponentPose=@n53, Alpha=1, GravityOverride="0,0,0", ComponentLinearAccScale="0.5,0.5,0.5", ComponentLinearVelScale="0.5,0.5,0.5") {
    angularBiasOverride=1
    angularDampingOverride=1
    angularSpringConstant=500
    bOverrideAngularBias=true
    bOverrideAngularDamping=true
    bUseGravityOverride=true
    simulationSpace=RootRelative
    alphaBoolBlend {blendOption=Linear}
    boundBone {boneName=dyn_rope_02}
    componentAppliedLinearAccClamp=<1000 1000 1000>
    componentLinearAccScale=<0.5 0.5 0.5>
    componentLinearVelScale=<0.5 0.5 0.5>
    physicsBodyDefinitions [{
      collisionType=CustomSphere
      sphereCollisionRadius=2
      boundBone {boneName=dyn_rope_02}
      boxExtents=<15 15 5>
      constraintSetup {angularLimitsMax=<0 15 0>, angularLimitsMin=<0 -45 0>, angularTarget=<1 0 0>}
      localJointOffset=<2 0 0>
     }]
    planarLimits [{drivingBone {boneName=spine_03_in}, planeTransform {location=<0 -15 0>, rotation=<rot 0 0 -90>}}]
   }
   n55 = Anim Dynamics - Bone: dyn_bow_01_L(ComponentPose=@n54, Alpha=1, GravityOverride="0,0,0", ComponentLinearAccScale="0.5,0.5,0.5", ComponentLinearVelScale="0.5,0.5,0.5") {
    angularBiasOverride=1
    angularDampingOverride=1
    bOverrideAngularBias=true
    bOverrideAngularDamping=true
    bUseGravityOverride=true
    bUsePlanarLimit=false
    simulationSpace=RootRelative
    alphaBoolBlend {blendOption=Linear}
    boundBone {boneName=dyn_bow_01_L}
    componentAppliedLinearAccClamp=<1000 1000 1000>
    componentLinearAccScale=<0.5 0.5 0.5>
    componentLinearVelScale=<0.5 0.5 0.5>
    physicsBodyDefinitions [{
      sphereCollisionRadius=1
      boundBone {boneName=dyn_bow_01_L}
      boxExtents=<15 15 15>
      constraintSetup {angularLimitsMax=<0 0 0>, angularLimitsMin=<0 -10 0>, angularTarget=<1 0 0>}
      localJointOffset=<2 0 0>
     }]
   }
   n56 = Anim Dynamics - Bone: dyn_bow(ComponentPose=@n55, Alpha=1, GravityOverride="0,0,0", ComponentLinearAccScale="0.5,0.5,0.5", ComponentLinearVelScale="0.5,0.5,0.5") {
    angularBiasOverride=1
    angularDampingOverride=1
    bOverrideAngularBias=true
    bOverrideAngularDamping=true
    bUseGravityOverride=true
    simulationSpace=RootRelative
    alphaBoolBlend {blendOption=Linear}
    boundBone {boneName=dyn_bow}
    componentAppliedLinearAccClamp=<1000 1000 1000>
    componentLinearAccScale=<0.5 0.5 0.5>
    componentLinearVelScale=<0.5 0.5 0.5>
    physicsBodyDefinitions [{
      collisionType=CustomSphere
      sphereCollisionRadius=2
      boundBone {boneName=dyn_bow}
      boxExtents=<10 10 5>
      constraintSetup {angularLimitsMax=<0 0 0>, angularLimitsMin=<0 -25 0>, angularTarget=<1 0 0>}
      localJointOffset=<1 0 0>
     }]
    planarLimits [{drivingBone {boneName=spine_03_in}, planeTransform {location=<0 -15 0>, rotation=<rot 0 0 -90>}}]
   }
   n57 = Anim Dynamics - Bone: dyn_bow_01_R(ComponentPose=@n56, Alpha=1, GravityOverride="0,0,0", ComponentLinearAccScale="0.5,0.5,0.5", ComponentLinearVelScale="0.5,0.5,0.5") {
    angularBiasOverride=1
    angularDampingOverride=1
    bOverrideAngularBias=true
    bOverrideAngularDamping=true
    bUseGravityOverride=true
    bUsePlanarLimit=false
    simulationSpace=RootRelative
    alphaBoolBlend {blendOption=Linear}
    boundBone {boneName=dyn_bow_01_R}
    componentAppliedLinearAccClamp=<1000 1000 1000>
    componentLinearAccScale=<0.5 0.5 0.5>
    componentLinearVelScale=<0.5 0.5 0.5>
    physicsBodyDefinitions [{
      sphereCollisionRadius=1
      boundBone {boneName=dyn_bow_01_R}
      boxExtents=<15 15 15>
      constraintSetup {angularLimitsMax=<0 0 0>, angularLimitsMin=<0 -10 0>, angularTarget=<1 0 0>}
      localJointOffset=<2 0 0>
     }]
   }
   n58 = Anim Dynamics - Bone: dyn_ribbon_01_L(ComponentPose=@n57, Alpha=1, GravityOverride="0,0,0", ComponentLinearAccScale="0.5,0.5,0.5", ComponentLinearVelScale="0.5,0.5,0.5") {
    angularBiasOverride=1
    angularDampingOverride=1
    bOverrideAngularBias=true
    bOverrideAngularDamping=true
    bUseGravityOverride=true
    simulationSpace=RootRelative
    alphaBoolBlend {blendOption=Linear}
    boundBone {boneName=dyn_ribbon_01_L}
    componentAppliedLinearAccClamp=<1000 1000 1000>
    componentLinearAccScale=<0.5 0.5 0.5>
    componentLinearVelScale=<0.5 0.5 0.5>
    physicsBodyDefinitions [{
      collisionType=CustomSphere
      sphereCollisionRadius=2
      boundBone {boneName=dyn_ribbon_01_L}
      boxExtents=<10 10 10>
      constraintSetup {angularLimitsMax=<0 0 10>, angularLimitsMin=<0 -25 0>, angularTarget=<1 0 0>}
      localJointOffset=<3 0 0>
     }]
    planarLimits [{drivingBone {boneName=spine_02_in}, planeTransform {location=<0 -13 0>, rotation=<rot 0 0 -90>}}]
   }
   n59 = Anim Dynamics - Bone: dyn_ribbon_01_R(ComponentPose=@n58, Alpha=1, GravityOverride="0,0,0", ComponentLinearAccScale="0.5,0.5,0.5", ComponentLinearVelScale="0.5,0.5,0.5") {
    angularBiasOverride=1
    angularDampingOverride=1
    bOverrideAngularBias=true
    bOverrideAngularDamping=true
    bUseGravityOverride=true
    simulationSpace=RootRelative
    alphaBoolBlend {blendOption=Linear}
    boundBone {boneName=dyn_ribbon_01_R}
    componentAppliedLinearAccClamp=<1000 1000 1000>
    componentLinearAccScale=<0.5 0.5 0.5>
    componentLinearVelScale=<0.5 0.5 0.5>
    physicsBodyDefinitions [{
      collisionType=CustomSphere
      sphereCollisionRadius=2
      boundBone {boneName=dyn_ribbon_01_R}
      boxExtents=<10 10 10>
      constraintSetup {angularLimitsMax=<0 0 0>, angularLimitsMin=<0 -25 -10>, angularTarget=<0 0 0>}
      localJointOffset=<3 0 0>
     }]
    planarLimits [{drivingBone {boneName=spine_02_in}, planeTransform {location=<0 -13 0>, rotation=<rot 0 0 -90>}}]
   }
   n60 = Anim Dynamics - Bone: dyn_ribbon_02_L(ComponentPose=@n59, Alpha=1, GravityOverride="0,0,0", ComponentLinearAccScale="0.5,0.5,0.5", ComponentLinearVelScale="0.5,0.5,0.5") {
    angularBiasOverride=1
    angularDampingOverride=1
    bOverrideAngularBias=true
    bOverrideAngularDamping=true
    bUseGravityOverride=true
    simulationSpace=RootRelative
    alphaBoolBlend {blendOption=Linear}
    boundBone {boneName=dyn_ribbon_02_L}
    componentAppliedLinearAccClamp=<1000 1000 1000>
    componentLinearAccScale=<0.5 0.5 0.5>
    componentLinearVelScale=<0.5 0.5 0.5>
    physicsBodyDefinitions [{
      collisionType=CustomSphere
      sphereCollisionRadius=2
      boundBone {boneName=dyn_ribbon_02_L}
      boxExtents=<10 10 10>
      constraintSetup {angularLimitsMax=<0 0 10>, angularLimitsMin=<0 -25 0>, angularTarget=<1 0 0>}
      localJointOffset=<4 0 0>
     }]
    planarLimits [{drivingBone {boneName=spine_02_in}, planeTransform {location=<0 -13 0>, rotation=<rot 0 0 -90>}}]
   }
   n61 = Anim Dynamics - Bone: dyn_ribbon_02_R(ComponentPose=@n60, Alpha=1, GravityOverride="0,0,0", ComponentLinearAccScale="0.5,0.5,0.5", ComponentLinearVelScale="0.5,0.5,0.5") {
    angularBiasOverride=1
    angularDampingOverride=1
    bOverrideAngularBias=true
    bOverrideAngularDamping=true
    bUseGravityOverride=true
    simulationSpace=RootRelative
    alphaBoolBlend {blendOption=Linear}
    boundBone {boneName=dyn_ribbon_02_R}
    componentAppliedLinearAccClamp=<1000 1000 1000>
    componentLinearAccScale=<0.5 0.5 0.5>
    componentLinearVelScale=<0.5 0.5 0.5>
    physicsBodyDefinitions [{
      collisionType=CustomSphere
      sphereCollisionRadius=2
      boundBone {boneName=dyn_ribbon_02_R}
      boxExtents=<10 10 10>
      constraintSetup {angularLimitsMax=<0 0 0>, angularLimitsMin=<0 -25 -10>, angularTarget=<0 0 0>}
      localJointOffset=<4 0 0>
     }]
    planarLimits [{drivingBone {boneName=spine_02_in}, planeTransform {location=<0 -13 0>, rotation=<rot 0 0 -90>}}]
   }
   Save cached pose 'male'(Pose=Blend Poses by bool(BlendPose_0=Component To Local(ComponentPose=@n61), BlendPose_1=Component To Local(ComponentPose=@n49), bActiveValue=Get Is Large()) {blendType=Linear}) {cacheName=male}
  ]}
 EventGraph {nodes [
   n1 = Event BlueprintUpdateAnimation()
   > Set Delta Time X(Delta Time X=@n1.DeltaTimeX)
   > n2 = Sequence!(execute) -> (then_0, then_1, then_2:Exec)
   > n3 = Gravity Override State Function DEPRECATED(
    CharacterPartAnimInstance=Self-Reference()
    GameTimeInSeconds=GetGameTimeInSeconds()
    GravityOverrideParams=Get Gravity Override Params()
    RBAN Params=Get RBAN Params()
   )
   > Set Gravity Override(Gravity Override=@n3)
   > Set RBAN Settings Out(RBAN Settings Out:RBANSetting Struct Structure=@n3.RBAN SettingsOut) -> (Output_Get:RBANSetting Struct Structure)
   > Set Skydive Floating(Skydive Floating:double=SelectFloat(A=1, B=0, bPickA=Get bIsSkydiveFloating())) -> (Output_Get:double)
   n4 = Branch(execute=@n2.then_1, Condition=Get bIsParachuteOpened())
   > n5 = Sequence!(execute) -> (then_0, then_1)
   > Set Back Pack Float Adjusted(Back Pack Float Adjusted:bool=false) -> (Output_Get:bool)
   n6 = Get Back Pack Parachuting adjusted() -> (Back Pack Parachuting adjusted:bool)
   n7 = Get PlayerGliderType() -> (PlayerGliderType=HangGlider)
   > n8 = Branch(Condition=AND Boolean(A=NOT Boolean(A=@n6), B="Equal (Enum)"(A:EGliderType Enum=@n7, B:EGliderType Enum=HangGlider)))
   > n9 = Sequence()
   > Do Once(Reset=@n10.then_0)
   > Delay(Duration=0.3)
   > Set Delay(Delay:bool=true) -> (Output_Get:bool)
   Branch(execute=@n9.then_1, Condition=Get Delay() -> (Delay:bool))
   > Set BackPack Corretion(BackPack Corretion:double=FInterpTo(Current=Get BackPack Corretion() -> (BackPack Corretion:double), Target=1, DeltaTime=Get Delta Time X(), InterpSpeed=5)) -> (Output_Get:double)
   > Branch(Condition="Equal (Float)"(A=Get BackPack Corretion(), B=1))
   > Set Back Pack Parachuting adjusted(Back Pack Parachuting adjusted:bool=true) -> (Output_Get:bool)
   Set BackPack Corretion(execute=@n8.else, BackPack Corretion=FInterpTo(Current=Get BackPack Corretion(), Target=0, DeltaTime=Get Delta Time X(), InterpSpeed=3))
   > Set Delay(Delay=false)
   Do Once(execute=@n11.then_0, Reset=@n5.then_1)
   > Set BackPack Corretion(BackPack Corretion=1)
   n10 = Sequence(execute=@n4.else)
   Set Back Pack Parachuting adjusted(execute=@n10.then_1, Back Pack Parachuting adjusted=false)
   > Set Delay(Delay=false)
   > Branch(Condition=NOT Boolean(A=Get Back Pack Float Adjusted() -> (Back Pack Float Adjusted:bool)))
   > n11 = Sequence()
   Set BackPack Corretion(execute=@n11.then_1, BackPack Corretion=FInterpTo(Current=Get BackPack Corretion(), Target=0, DeltaTime=Get Delta Time X(), InterpSpeed=2))
   > Branch(Condition="Equal (Float)"(A=Get BackPack Corretion()))
   > Set Back Pack Float Adjusted(Back Pack Float Adjusted=true)
   n12 = Get BodyMesh() -> (BodyMesh:Skeletal Mesh Component Object Reference)
   Is Valid(exec=@n2.then_2, InputObject=@n12)
   n13 = BreakTransform(InTransform=GetSocketTransform(self=@n12, InSocketName=pelvis, TransformSpace=RTS_World))
   > Set pelvis translation(pelvis translation:Vector=@n13) -> (Output_Get:Vector)
   > Set pelvis rotation(pelvis rotation:Rotator=@n13.Rotation) -> (Output_Get:Rotator)
   n14 = BreakTransform(InTransform=GetSocketTransform(self=@n12, InSocketName=spine_01, TransformSpace=RTS_World))
   > Set spine 01 translation(spine 01 translation:Vector=@n14) -> (Output_Get:Vector)
   > Set spine 01 rotation(spine 01 rotation:Rotator=@n14.Rotation) -> (Output_Get:Rotator)
   n15 = BreakTransform(InTransform=GetSocketTransform(self=@n12, InSocketName=spine_02, TransformSpace=RTS_World))
   > Set spine 02 translation(spine 02 translation:Vector=@n15) -> (Output_Get:Vector)
   > Set spine 02 rotation(spine 02 rotation:Rotator=@n15.Rotation) -> (Output_Get:Rotator)
   n16 = BreakTransform(InTransform=GetSocketTransform(self=@n12, InSocketName=spine_03, TransformSpace=RTS_World))
   > Set spine 03 translation(spine 03 translation:Vector=@n16) -> (Output_Get:Vector)
   > Set spine 03 rotation(spine 03 rotation:Rotator=@n16.Rotation) -> (Output_Get:Rotator)
   n17 = BreakTransform(InTransform=GetSocketTransform(self=@n12, InSocketName=spine_04, TransformSpace=RTS_World))
   > Set spine 04 translation(spine 04 translation:Vector=@n17) -> (Output_Get:Vector)
   > Set spine 04 rotation(spine 04 rotation:Rotator=@n17.Rotation) -> (Output_Get:Rotator)
   n18 = GetDeltaTransformFromRefPose!(self=@n12, BoneName=spine_05, BaseName=pelvis) -> (
    ReturnValue_Location:Vector
    ReturnValue_Rotation_Roll:float
    ReturnValue_Rotation_Pitch:float
    ReturnValue_Rotation_Yaw:float
    ReturnValue_Scale:Vector
   )
   > Set Spine_05 Roll(Spine_05 Roll:double=MapRangeClamped(Value=@n18.ReturnValue_Rotation_Roll, InRangeA=0, InRangeB=-15)) -> (Output_Get:double)
   > Set Spine_05 Yaw(Spine_05 Yaw:double=MapRangeClamped(Value=@n18.ReturnValue_Rotation_Yaw, InRangeA=10, InRangeB=20)) -> (Output_Get:double)
   n19 = Event OnCharacterCustomizationLoaded()
   > Set BodyMesh(BodyMesh=Get Array Item with EFort Custom Part Type(Array=@n19.CustomMeshParts))
   n20 = Get Array Item with EFort Custom Part Type(Array:Array of Custom Character Part Object References=@n19.CustomParts) -> (Output:Custom Character Part Object Reference)
   > Set Is Large(Is Large:bool="Equal (Enum)"(
     A:EFortCustomBodyType Enum=Get BodyTypesPermitted!(self:Custom Character Part Object Reference=@n20) -> (BodyTypesPermitted:EFortCustomBodyType Enum=NONE)
     B:EFortCustomBodyType Enum=Large
    )) -> (Output_Get:bool)
   > Set isMale(isMale:bool="Equal (Enum)"(
     A:EFortCustomGender Enum=Get GenderPermitted!(self:Custom Character Part Object Reference=@n20) -> (GenderPermitted:EFortCustomGender Enum=Invalid)
     B:EFortCustomGender Enum=Male
    )) -> (Output_Get:bool)
   TryGetPawnOwner()
  ]}
]
variables [
 Back Pack Float Adjusted {type=boolean}
 Back Pack Parachuting adjusted {type=boolean}
 BackPack Corretion {type=number}
 BodyMesh {type=SkeletalMeshComponent}
 Delay {type=boolean}
 Delta Time X {type=number}
 Gravity Override {type=Vector}
 Gravity Override Params {type=GravityOverrideParamsStruct}
 Is Large {type=boolean}
 isMale {type=boolean}
 pelvis rotation {type=Rotator}
 pelvis translation {type=Vector}
 RBAN Params {
  type=RBANParamsStruct
  default {
   dBNO {
    componentLinearAccScale=<0.5 0.5 0.5>
    componentLinearVelScale=<0.5 0.5 0.5>
    simSpaceSettings {
     maxAngularAcceleration=1000
     maxAngularVelocity=3
     maxLinearAcceleration=300
     maxLinearVelocity=700
     velocityScaleZ=0
     externalLinearDragV=<0 0.5 0>
    }
   }
   dBNO_Carried {
    componentLinearAccScale=<0.5 0.5 0.5>
    componentLinearVelScale=<0.5 0.5 0.5>
    simSpaceSettings {
     maxAngularAcceleration=1000
     maxAngularVelocity=3
     maxLinearAcceleration=300
     maxLinearVelocity=700
     velocityScaleZ=0
     externalLinearDragV=<0 0.5 0>
    }
   }
   emoteOrMelee {
    componentLinearAccScale=<0.5 0.5 0.5>
    componentLinearVelScale=<0.5 0.5 0.5>
    simSpaceSettings {
     maxAngularAcceleration=1000
     maxAngularVelocity=3
     maxLinearAcceleration=300
     maxLinearVelocity=700
     velocityScaleZ=0
     externalLinearDragV=<0 0.5 0>
    }
   }
   front_End {
    componentLinearAccScale=<0.1 0.1 0.1>
    componentLinearVelScale=<0.1 0.1 0.1>
    simSpaceSettings {
     maxAngularAcceleration=1000
     maxAngularVelocity=3
     maxLinearAcceleration=300
     maxLinearVelocity=700
     velocityScaleZ=0
     externalLinearDragV=<0 0.5 0>
    }
   }
   onGround_Crouch_Moving {
    componentLinearAccScale=<0.5 0.5 0.5>
    componentLinearVelScale=<0.5 0.5 0.5>
    simSpaceSettings {
     maxAngularAcceleration=1000
     maxAngularVelocity=3
     maxLinearAcceleration=300
     maxLinearVelocity=700
     velocityScaleZ=0
     externalLinearDragV=<0 0.5 0>
    }
   }
   onGround_Falling {
    componentLinearAccScale=<0.5 0.5 0.5>
    componentLinearVelScale=<0.5 0.5 0.5>
    simSpaceSettings {
     maxAngularAcceleration=1000
     maxAngularVelocity=3
     maxLinearAcceleration=300
     maxLinearVelocity=700
     velocityScaleZ=0
     externalLinearDragV=<0 0.5 0>
    }
   }
   onGround_Moving {
    componentLinearAccScale=<0.5 0.5 0.5>
    componentLinearVelScale=<0.5 0.5 0.5>
    simSpaceSettings {
     maxAngularAcceleration=1000
     maxAngularVelocity=3
     maxLinearAcceleration=300
     maxLinearVelocity=700
     velocityScaleZ=0
     externalLinearDragV=<0 0.5 0>
    }
   }
   onGround_Standing {
    componentLinearAccScale=<0.5 0.5 0.5>
    componentLinearVelScale=<0.5 0.5 0.5>
    simSpaceSettings {
     maxAngularAcceleration=1000
     maxAngularVelocity=3
     maxLinearAcceleration=300
     maxLinearVelocity=700
     velocityScaleZ=0
     externalLinearDragV=<0 0.5 0>
    }
   }
   rail_Grinding {
    componentLinearAccScale=<0.5 0.5 0.5>
    componentLinearVelScale=<0.5 0.5 0.5>
    simSpaceSettings {
     maxAngularAcceleration=1000
     maxAngularVelocity=3
     maxLinearAcceleration=300
     maxLinearVelocity=700
     velocityScaleZ=0
     worldAlpha=1
     externalLinearDragV=<0 0.5 0>
    }
   }
   riding_Moving {
    componentLinearAccScale=<0.5 0.5 0.5>
    componentLinearVelScale=<0.5 0.5 0.5>
    simSpaceSettings {
     maxAngularAcceleration=1000
     maxAngularVelocity=3
     maxLinearAcceleration=300
     maxLinearVelocity=700
     velocityScaleZ=0
     externalLinearDragV=<0 0.5 0>
    }
   }
   riding_Moving_Targeting {
    componentLinearAccScale=<0.5 0.5 0.5>
    componentLinearVelScale=<0.5 0.5 0.5>
    simSpaceSettings {
     maxAngularAcceleration=1000
     maxAngularVelocity=3
     maxLinearAcceleration=300
     maxLinearVelocity=700
     velocityScaleZ=0
     externalLinearDragV=<0 0.5 0>
    }
   }
   riding_Standing {
    componentLinearAccScale=<0.5 0.5 0.5>
    componentLinearVelScale=<0.5 0.5 0.5>
    simSpaceSettings {
     maxAngularAcceleration=1000
     maxAngularVelocity=3
     maxLinearAcceleration=300
     maxLinearVelocity=700
     velocityScaleZ=0
     externalLinearDragV=<0 0.5 0>
    }
   }
   skydive_DivingDown {
    componentLinearAccScale=<0.5 0.5 0.5>
    componentLinearVelScale=<0.5 0.5 0.5>
    simSpaceSettings {
     maxAngularAcceleration=1000
     maxAngularVelocity=3
     maxLinearAcceleration=300
     maxLinearVelocity=700
     velocityScaleZ=0
     externalLinearDragV=<0 0.5 0>
    }
   }
   skydive_FloatingDown {
    componentLinearAccScale=<0.5 0.5 0.5>
    componentLinearVelScale=<0.5 0.5 0.5>
    simSpaceSettings {
     maxAngularAcceleration=1000
     maxAngularVelocity=3
     maxLinearAcceleration=300
     maxLinearVelocity=700
     velocityScaleZ=0
     externalLinearDragV=<0 0.5 0>
    }
   }
   skydive_GoingUp {
    componentLinearAccScale=<0.5 0.5 0.5>
    componentLinearVelScale=<0.5 0.5 0.5>
    simSpaceSettings {
     maxAngularAcceleration=1000
     maxAngularVelocity=3
     maxLinearAcceleration=300
     maxLinearVelocity=700
     velocityScaleZ=0
     externalLinearDragV=<0 0.5 0>
    }
   }
   skydive_Parachuting {
    componentLinearAccScale=<0.5 0.5 0.5>
    componentLinearVelScale=<0.5 0.5 0.5>
    simSpaceSettings {
     maxAngularAcceleration=1000
     maxAngularVelocity=3
     maxLinearAcceleration=300
     maxLinearVelocity=700
     velocityScaleZ=0
     externalLinearDragV=<0 0.5 0>
    }
   }
   skydive_Surfing {
    componentLinearAccScale=<0.5 0.5 0.5>
    componentLinearVelScale=<0.5 0.5 0.5>
    simSpaceSettings {
     maxAngularAcceleration=1000
     maxAngularVelocity=3
     maxLinearAcceleration=300
     maxLinearVelocity=700
     velocityScaleZ=0
     externalLinearDragV=<0 0.5 0>
    }
   }
   swimming {
    componentLinearAccScale=<0.5 0.5 0.5>
    componentLinearVelScale=<0.5 0.5 0.5>
    simSpaceSettings {
     maxAngularAcceleration=1000
     maxAngularVelocity=3
     maxLinearAcceleration=300
     maxLinearVelocity=700
     velocityScaleZ=0
     externalLinearDragV=<0 0.5 0>
    }
   }
   tactical_Sprint {
    componentLinearAccScale=<0.5 0.5 0.5>
    componentLinearVelScale=<0.5 0.5 0.5>
    simSpaceSettings {
     maxAngularAcceleration=1000
     maxAngularVelocity=3
     maxLinearAcceleration=300
     maxLinearVelocity=700
     velocityScaleZ=0
     externalLinearDragV=<0 0.5 0>
    }
   }
   ziplining {
    componentLinearAccScale=<0.5 0.5 0.5>
    componentLinearVelScale=<0.5 0.5 0.5>
    simSpaceSettings {
     maxAngularAcceleration=1000
     maxAngularVelocity=3
     maxLinearAcceleration=300
     maxLinearVelocity=700
     velocityScaleZ=0
     externalLinearDragV=<0 0.5 0>
    }
   }
  }
 }
 RBAN Settings Out {
  type=RBANSettingStruct
  default {
   componentLinearAccScale=<0 0 0>
   componentLinearVelScale=<0 0 0>
   simSpaceSettings {
    dampingAlpha=1
    maxAngularAcceleration=1000
    maxAngularVelocity=3
    maxLinearAcceleration=300
    maxLinearVelocity=700
    velocityScaleZ=0
    worldAlpha=1
    externalAngularVelocity=<0 0 0>
    externalLinearDragV=<0 0.5 0>
    externalLinearVelocity=<0 0 0>
   }
  }
 }
 Skydive Floating {type=number}
 spine 01 rotation {type=Rotator}
 spine 01 translation {type=Vector}
 spine 02 rotation {type=Rotator}
 spine 02 translation {type=Vector}
 spine 03 rotation {type=Rotator}
 spine 03 translation {type=Vector}
 spine 04 rotation {type=Rotator}
 spine 04 translation {type=Vector}
 Spine_05 Roll {type=number}
 Spine_05 Yaw {type=number}
]
